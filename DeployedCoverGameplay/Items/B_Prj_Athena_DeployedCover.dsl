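; Thrown projectile for the Deployed Cover item: it finds a valid landing spot, then spawns multi-piece cover there on the authority.
; Hits clamp bounce velocity; validation rejects moving platforms, queries DeployedCoverBPLibrary using start yaw, and optionally destroys overlapping props or launches pawns away.
; Chunks spawn relative to that point, optionally attaching unparented pieces to their base or the center piece.
; A projectile-stop failsafe reverts to a pickup when no valid spot exists.
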(event Collision|EventHit
  (params (MyComp :type Primitive%20Component%20Object%20Reference) (Other :type Actor%20Object%20Reference) (OtherComp :type Primitive%20Component%20Object%20Reference) (bSelfMoved :type Boolean) (HitLocation :type Vector :default "0, 0, 0") (HitNormal :type Vector :default "0, 0, 0") (NormalImpulse :type Vector :default "0, 0, 0") (Hit :type Hit%20Result%20Structure%20%28by%20ref%29))
  (bind _projectilemovementcomponent (Variables|Movement|GetProjectileMovementComponent))
  ;[1 1 1] Prevents in-air-loop from persisting after projectile lands
  (Audio|Components|Audio|Stop (Variables|Default|GetInAirLoopSound))
  (|CallParentFunction MyComp Other OtherComp bSelfMoved HitLocation HitNormal NormalImpulse Hit)
  ;[0.1771 0.6602 1] Clamp the velocity after the bounce to avoid flying 40m away
  (Projectile|SetVelocity _projectilemovementcomponent (Math|Vector|ClampVectorSize (Variables|/Script/Engine.MovementComponent|GetVelocity _projectilemovementcomponent) 0.0 1e+03))
  (Networking|CanExecuteCosmeticEvents
    (:False
      (EventDispatchers|BindEventtoOnProjectileStop _projectilemovementcomponent (EventDispatchers|CreateEvent :!event Custom|RevertToPickup_TimeOut)) ; Failsafe for projectile never finding a valid spawn location to revert back to a pickup.
    )))

(event Custom|RevertToPickup_TimeOut
  (params (ImpactResult :type Hit%20Result%20Structure%20%28by%20ref%29))
  ; Handle projectile stop bind. Revert to a pickup if projectile stopped and no valid spawn location was found.
  (Networking|SwitchHasAuthority
    (:Authority
      (if (not (Math|Boolean|ORBoolean (Variables|Spawning|GetSpawned) (Variables|Default|GetRevertedToPickup) (Variables|Default|GetbIsValidSpawnLocation)))
        (Variables|Default|SetRevertedToPickup true)
        (bind _itemdeftospawnaspickupwheninvalid (Variables|Spawning|GetItemDefToSpawnAsPickupWhenInvalid)
          (:then
            (bind _returnvalue (Inventory|SpawnPickupInWorld _itemdeftospawnaspickupwheninvalid 1 (+ (Transformation|GetActorLocation) "0.000000,0.000000,50.000000") "0.000000,0.000000,1.000000" 0 true false false -1 "Tossed" "TossedByPlayer" 0 false))
            (if (Variables|MarkItem|GetMarkItemPickup)
              (CallFunction|Mark _returnvalue)
              (else
                _))
            (Rendering|SetActorHiddenInGame (Variables|ReverttoObject|GetSetActorHiddenOnRevertToObject))
            (Collision|SetActorEnableCollision false)
            (Game|Components|ProjectileMovement|StopSimulating (Variables|Movement|GetProjectileMovementComponent) ImpactResult)
            (Actor|DestroyActor))
          (:else))))))

(event EventDestroyed
  (|CallParentFunction)
  (Networking|CanExecuteCosmeticEvents
    (:False
      (EventDispatchers|BindEventtoOnProjectileStop (Variables|Movement|GetProjectileMovementComponent) (EventDispatchers|CreateEvent :!event Custom|RevertToPickup_TimeOut)))))

(event EventBeginPlay
  (|CallParentFunction)
  (Variables|Spawning|ValidityChecks|SetMinDotWithUp (Math|Trig|Cos(Degrees) (ScalableFloat|GetValueasDouble (Variables|Default|GetHF_WallTiltMaxAngle))))
  (bind _asfort_player_pawn_athena (Utilities|Casting|CastToFortPlayerPawnAthena (Actor|GetOwner)))
  (bind (_returnvalue_roll _returnvalue_pitch _returnvalue_yaw) (Math|Rotator|BreakRotator (Pawn|GetControlRotation _asfort_player_pawn_athena :!class /Script/Engine.Pawn)))
  (Variables|Default|SetStartPlayerYaw _returnvalue_yaw)
  (Variables|Spawning|ValidityChecks|SetbDontSpawnOnPhysicsSim (ScalableFloat|GetValueasBoolwithContext (Variables|Default|GetHF_DisallowSpawnOnPhysicsSim))))

(fn B_Prj_Athena_DeployedCover_ActorsAtHitLocation
  (params (HitActor :type Actor%20Object%20Reference) (ImpactLocation :type Vector))
  (Utilities|IsA(soft) HitActor "/Script/ItemizedPropSpawnerParentRuntime.ItemizedPropSpawnerWallChunk"
    (:False
      (Utilities|FlowControl|Sequence
        (:then_0
          (if (ScalableFloat|GetValueasBoolwithContext (Variables|Default|GetHF_DestructionEnabled))
            (bind _asbuilding_prop (Utilities|Casting|CastToBuildingProp HitActor))
            (Utilities|IsValid (Ability|GetFortAbilitySystemComponent _asbuilding_prop)
              (:Is%20Valid
                (if (Math|Boolean|ANDBoolean (Variables|/Script/Engine.Actor|GetbCanBeDamaged _asbuilding_prop) (or (Variables|/Script/FortniteGame.BuildingActor|GetbDestroyOnPlayerBuildingPlacement _asbuilding_prop) (<= (Health|GetMaxHealth _asbuilding_prop :!class /Script/FortniteGame.BuildingActor) (ScalableFloat|GetValueatLevelwithContext (Variables|Default|GetHF_MaxPropHealthToDestroy)))) (GameplayTags|HasAnyMatchingGameplayTags (Ability|GetFortAbilitySystemComponent _asbuilding_prop) (GameplayTags|MakeLiteralGameplayTagContainer "(GameplayTags=((TagName=\"Building.Type.Prop\"),(TagName=\"Athena.Quests.Building.Wood.Fences\"),(TagName=\"Building.Type.Container.Slurp\")))")) (or (not (GameplayTags|HasAnyMatchingGameplayTags (Ability|GetFortAbilitySystemComponent _asbuilding_prop) (GameplayTags|MakeLiteralGameplayTagContainer "(GameplayTags=((TagName=\"Building.Type.ItemContainer\"),(TagName=\"Building.Type.Container\"),(TagName=\"Building.Type.Prop.Hiding\")))"))) (GameplayTags|HasAnyMatchingGameplayTags (Ability|GetFortAbilitySystemComponent _asbuilding_prop) (GameplayTags|MakeLiteralGameplayTagContainer "(GameplayTags=((TagName=\"Building.Type.Container.Organic\"),(TagName=\"Building.Type.Prop.Organic\"),(TagName=\"Building.Type.Container.Slurp.Barrel\")))"))))
                  (Utilities|IsValid _asbuilding_prop
                    (:Is%20Valid
                      (Actor|DestroyActor _asbuilding_prop))))))))
        (:then_1
          (if (ScalableFloat|GetValueasBoolwithContext (Variables|Default|GetHF_AdjustOverlappingPawns))
            (bind _asfort_player_pawn (Utilities|Casting|CastToFortPlayerPawn HitActor))
            (Components|Movement|StopMovementImmediately (Variables|/Script/Engine.Character|GetCharacterMovement _asfort_player_pawn))
            (Character|LaunchCharacter _asfort_player_pawn (Math|Vector|MakeVector (.x (* (Math|Vector|Normalize (- (Transformation|GetActorLocation _asfort_player_pawn) (Variables|GetImpactLocation))) (ScalableFloat|GetValueatLevelwithContext (Variables|Default|GetHF_PawnLaunchStrength)))) (.y (* (Math|Vector|Normalize (- (Transformation|GetActorLocation _asfort_player_pawn) (Variables|GetImpactLocation))) (ScalableFloat|GetValueatLevelwithContext (Variables|Default|GetHF_PawnLaunchStrength)))) (Math|Float|Max(Float) (+ (.z (* (Math|Vector|Normalize (- (Transformation|GetActorLocation _asfort_player_pawn) (Variables|GetImpactLocation))) (ScalableFloat|GetValueatLevelwithContext (Variables|Default|GetHF_PawnLaunchStrength)))) 5e+01) 5e+02)) true true)))))))

(fn IsHitValidForSpawn
  (params (HitOtherComp :type Primitive%20Component%20Object%20Reference) (HitOther :type Actor%20Object%20Reference) (HitResult :type Hit%20Result%20Structure))
  (bind (_bblockinghit _binitialoverlap _time _distance _location _impactpoint) (Collision|BreakHitResult (Variables|GetHitResult)))
  (bind (_bblockinghit_2 _binitialoverlap_2 _time_2 _distance_2 _location_2 _impactpoint_2) (Collision|BreakHitResult (Variables|GetHitResult)))
  (bind (_validspawn _reverttopickup) (|CallParentFunction HitOtherComp HitOther HitResult))
  (if (not (Utilities|IsValid (Class|FortBaseMovingPlatformComponent|FindMovingPlatformComponentfromActor HitOther true) :!class /Script/Engine.KismetSystemLibrary))
    (LocalVariables|SetShouldRevertToPickup _reverttopickup)
    (Utilities|Array|Clear (Variables|Default|GetSpawnValidities))
    (Variables|Default|SetSpawnRotation "0, 0, 0")
    (Utilities|FlowControl|Sequence
      (:then_0
        (if (and _validspawn (not (LocalVariables|GetShouldRevertToPickup)))
          (bind _returnvalue (CallFunction|GetOwnerWeapon))
          (bind _asb_deployed_cover_grenade (Utilities|Casting|CastToB_DeployedCoverGrenade _returnvalue))
          (bind (_isvalid _spawnvalidity _spawnrotation _spawnlocation) (Class|DeployedCoverBPLibrary|CheckSpawnValidityforDeployedCover (Variables|/DeployedCoverGameplay/Items/ItemSpawner/B_DeployedCoverGrenade.B_DeployedCoverGrenade_C|GetNativeData _asb_deployed_cover_grenade) _impactpoint (Variables|Default|GetStartPlayerYaw)))
          (Variables|Default|SetSpawnValidities _spawnvalidity)
          (Utilities|FlowControl|Sequence
            (:then_0
              ; Check for actors to destroy when overlapping with the cover. Do it before spawning to avoid parenting to an actor that will be destroyed.
              (if (or (ScalableFloat|GetValueasBoolwithContext (Variables|Default|GetHF_DestructionEnabled)) (ScalableFloat|GetValueasBoolwithContext (Variables|Default|GetHF_AdjustOverlappingPawns)))
                (for _array_element (Variables|Default|GetSpawnValidities)
                  (if (Utilities|Struct|BreakDeployedCoverSpawnInfo _array_element)
                    (for _array_element_2 (.OverlappingActors (Utilities|Struct|BreakDeployedCoverSpawnInfo _array_element))
                      (CallFunction|BPrjAthenaDeployedCoverActorsatHitLocation _array_element_2 _impactpoint_2))))))
            (:then_1
              (Variables|Default|SetbIsValidSpawnLocation _isvalid) ; Set valid spawn location to check on projectile stop bind to know if we should revert to pickup or let spawn.
              (Variables|Default|SetSpawnLocation _spawnlocation)
              (Variables|Default|SetSpawnRotation _spawnrotation))
            (:then_2
              (return (Variables|Default|GetbIsValidSpawnLocation) (LocalVariables|GetShouldRevertToPickup))))))
      (:then_1
        (return _validspawn (LocalVariables|GetShouldRevertToPickup))))
    (else
      (return false true))))

(fn SpawnAdditionalBGA
  (params (HitNormal :type Vector) (HitActor :type Actor%20Object%20Reference) (HitComp :type Primitive%20Component%20Object%20Reference) (HitPosition :type Vector) (FromWater :type Boolean) (HitData :type Hit%20Result%20Structure))
  (bind _deployedcoverweapon (LocalVariables|GetDeployedCoverWeapon))
  (bind _nativedata (Variables|/DeployedCoverGameplay/Items/ItemSpawner/B_DeployedCoverGrenade.B_DeployedCoverGrenade_C|GetNativeData _deployedcoverweapon))
  (bind (_player _weapon _bpreviewing _previewdistance _maxdistance _mindistance _mindistfromindestructible _benableblockedbyindestructibles _playerforwardvectorscale _initialtracepullbackdist _downwardtracedistancenear _downwardtracedistancenearup _downwardtracedistancefar _bstopped _classestoignorewhentracing _classestoignorewhentracingiftoosmall _debugdrawtime _chunkrelativetransforms _chunklocationtochunkindexmap) (Utilities|Struct|BreakItemizedPropSpawnerWeaponData _nativedata))
  (bind _deployedcoverweapon_2 (LocalVariables|GetDeployedCoverWeapon))
  (bind _nativedata_2 (Variables|/DeployedCoverGameplay/Items/ItemSpawner/B_DeployedCoverGrenade.B_DeployedCoverGrenade_C|GetNativeData _deployedcoverweapon_2))
  (bind (_player_2 _weapon_2 _bpreviewing_2 _previewdistance_2 _maxdistance_2 _mindistance_2 _mindistfromindestructible_2 _benableblockedbyindestructibles_2 _playerforwardvectorscale_2 _initialtracepullbackdist_2 _downwardtracedistancenear_2 _downwardtracedistancenearup_2 _downwardtracedistancefar_2 _bstopped_2 _classestoignorewhentracing_2 _classestoignorewhentracingiftoosmall_2 _debugdrawtime_2 _chunkrelativetransforms_2 _chunklocationtochunkindexmap_2) (Utilities|Struct|BreakItemizedPropSpawnerWeaponData _nativedata_2))
  (bind _hadditionalspawnedactors (LocalVariables|GethAdditionalSpawnedActors))
  (bind _spawnvalidities (Variables|Default|GetSpawnValidities))
  (Utilities|FlowControl|Sequence
    (:then_0
      (Networking|SwitchHasAuthority
        (:Authority
          (if (not (Variables|Default|GetRevertedToPickup))
            (bind _returnvalue (CallFunction|GetOwnerWeapon))
            (bind _asb_deployed_cover_grenade (Utilities|Casting|CastToB_DeployedCoverGrenade _returnvalue))
            (LocalVariables|SetDeployedCoverWeapon _asb_deployed_cover_grenade)
            (Utilities|FlowControl|Sequence
              (:then_0
                (for _array_element (Variables|/ItemizedPropSpawnerParent/Blueprints/DataClass_ItemizedPropSpawnerData.DataClass_ItemizedPropSpawnerData_C|GetNativePropsToSpawn (Variables|/DeployedCoverGameplay/Items/ItemSpawner/B_DeployedCoverGrenade.B_DeployedCoverGrenade_C|GetData%20Asset (LocalVariables|GetDeployedCoverWeapon)))
                  (if (.ReturnValue (Utilities|Map|Find _chunklocationtochunkindexmap_2 (.ChunkLocation (Utilities|Struct|BreakNativeItemizedPropSpawnerPropSpawnData _array_element))))
                    (LocalVariables|SetCurrentChunkIndex (Utilities|Map|Find _chunklocationtochunkindexmap_2 (.ChunkLocation (Utilities|Struct|BreakNativeItemizedPropSpawnerPropSpawnData _array_element))))
                    (if (Utilities|Struct|BreakDeployedCoverSpawnInfo (Utilities|Array|Get(aref) (Variables|Default|GetSpawnValidities) (LocalVariables|GetCurrentChunkIndex)))
                      ;[0.4426 1 0.5749] Spawn the additional BGA based on the central piece position and rotation
                      (bind _returnvalue_2 (Building|SpawnBuildingActor (Utilities|Struct|BreakNativeItemizedPropSpawnerPropSpawnData _array_element) (Math|Transform|ComposeTransforms (.RelativeTransform (Utilities|Struct|BreakNativeItemizedPropSpawnerPropSpawnData _array_element)) (Math|Transform|MakeTransform (Variables|Default|GetSpawnLocation) (Variables|Default|GetSpawnRotation))) (Variables|Actor|GetInstigator) (Variables|Actor|GetInstigator)))
                      (Utilities|Map|Add (LocalVariables|GethAdditionalSpawnedActors) (LocalVariables|GetCurrentChunkIndex) _returnvalue_2)
                      (if (Utilities|Enum|Equal(Enum) (.ChunkLocation (Utilities|Struct|BreakNativeItemizedPropSpawnerPropSpawnData _array_element)) "Center")
                        (LocalVariables|SetCenterBuild _returnvalue_2))
                      (Building|InitializeKismetSpawnedBuildingActor _returnvalue_2 0 (select (Variables|Spawning|GetUsePlayerAsSpawningController) (Projectile|GetInstigatorPlayerController) 0) true 0 false)
                      ;; Manually set PlayerPlaced for train/grind to destroy the cover
                      (BuildingActor|SetPlayerPlaced _returnvalue_2 true)
                      (bind _returnvalue_3 (CallFunction|GetOwnerWeapon))
                      (Utilities|IsValid _returnvalue_3
                        (:Is%20Valid
                          (CosmeticWraps|AttemptApplyOverrideWraptoActor _returnvalue_3 _returnvalue_2)))))))
              (:then_1
                ; Adjust cover attachment
                (if (ScalableFloat|GetValueasBool (Variables|Default|GetHF_AdjustAttachBase))
                  (for _array_element_2 (LocalVariables|GetChunkLocationsToLoop)
                    (if (.ReturnValue (Utilities|Map|Find _chunklocationtochunkindexmap _array_element_2))
                      (if (.ReturnValue (Utilities|Map|Find _hadditionalspawnedactors (Utilities|Map|Find _chunklocationtochunkindexmap _array_element_2)))
                        (LocalVariables|SetCurrentBGAToProcess (Utilities|Map|Find _hadditionalspawnedactors (Utilities|Map|Find _chunklocationtochunkindexmap _array_element_2)))
                        (bind _returnvalue_4 (Attachment|GetParentBuildingAttachment(Message) (LocalVariables|GetCurrentBGAToProcess)))
                        (Utilities|IsValid _returnvalue_4
                          (:Is%20Not%20Valid
                            (LocalVariables|SetCurrentBaseBGA 0)
                            (Utilities|FlowControl|Sequence
                              (:then_0
                                (bind _asbuilding_static_mesh (Utilities|Casting|CastToBuildingSMActor (.BaseActor (Utilities|Struct|BreakDeployedCoverSpawnInfo (Utilities|Array|Get(aref) _spawnvalidities (Utilities|Map|Find _chunklocationtochunkindexmap _array_element_2)))))
                                  (:then
                                    (LocalVariables|SetCurrentBaseBGA _asbuilding_static_mesh))
                                  (:CastFailed
                                    (Utilities|Casting|CastToLandscape (.BaseActor (Utilities|Struct|BreakDeployedCoverSpawnInfo (Utilities|Array|Get(aref) _spawnvalidities (Utilities|Map|Find _chunklocationtochunkindexmap _array_element_2))))
                                      (:CastFailed
                                        (bind _asbuilding_static_mesh_2 (Utilities|Casting|CastToBuildingSMActor (LocalVariables|GetCenterBuild)))
                                        (LocalVariables|SetCurrentBaseBGA _asbuilding_static_mesh_2))) ; assume landscape to be always valid base
                                  )))
                              (:then_1
                                (Utilities|IsValid (LocalVariables|GetCurrentBaseBGA)
                                  (:Is%20Valid
                                    (Attachment|AttachBuildingActortoMe (LocalVariables|GetCurrentBaseBGA) (LocalVariables|GetCurrentBGAToProcess) true)))))))))))))))))
    (:then_1
      ;[1 0.5165 0.7519] Return
      ; sequence ensures this always executes
      (return (LocalVariables|GethAdditionalSpawnedActors) (LocalVariables|GetCenterBuild)))))

(fn SpawnBGA_
  (params (Normal :type Vector :default "0, 0, 0") (HitActor :type Actor%20Object%20Reference) (HitComp :type Primitive%20Component%20Object%20Reference) (Pos :type Vector :default "0, 0, 0") (FromWater :type Boolean) (Hit :type Hit%20Result%20Structure))
  (|CallParentFunction Normal HitActor HitComp Pos FromWater Hit)
  (bind (_spawnedactors _spawnedcenter) (CallFunction|SpawnAdditionalBGA Normal HitActor HitComp Pos FromWater Hit))
  (Utilities|Array|Clear (Variables|Default|GetSpawnValidities))
  (return _spawnedcenter))

(fn ConstructionScript
  (|CallParentFunction))
