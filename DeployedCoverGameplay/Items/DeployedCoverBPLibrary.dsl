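; Blueprint function library for the Deployed Cover grenade that checks whether cover can spawn at a target location and yaw.
; CheckSpawnValidityForDeployedCover packs ScalableFloat tuning, a blocking-area tag, a trace channel, the grenade data asset's props and a fixed class list into spawner data.
; It passes wall tilt as a cosine and delegates to DeployedCoverFunctionLibrary's CheckSpawnValidity, returning validity, spawn info, rotation and location.
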(fn CheckSpawnValidityForDeployedCover
  (params (WeaponData :type Itemized%20Prop%20Spawner%20Weapon%20Data%20Structure) (TargetLocation :type Vector) (YawRotation :type Float%20%28double-precision%29) (Debug :type Boolean))
  (bind (_player _weapon) (Utilities|Struct|BreakItemizedPropSpawnerWeaponData (Variables|GetWeaponData)))
  (bind _asb_deployed_cover_grenade (Utilities|Casting|CastToB_DeployedCoverGrenade _weapon))
  (bind (_boutcanspawn _outspawnlocation _outspawnrotation _outhits _outspawninfo) (Class|DeployedCoverFunctionLibrary|CheckSpawnValidity (Variables|GetWeaponData) (Utilities|Struct|MakeDeployedCoverSpawnerData (ScalableFloat|GetValueasBool (LocalVariables|GetSF_LineTraceForSpawnRot)) (ScalableFloat|GetValueasBool (LocalVariables|GetSF_AdjustAttachBase)) (ScalableFloat|GetValueasBool (LocalVariables|GetSF_DontSpawnOnPlayers)) (ScalableFloat|GetValueasBool (LocalVariables|GetSF_DontSpawnOnPhysicsSim)) (ScalableFloat|GetValueatLevel (LocalVariables|GetSF_TraceCapsuleRadius)) (ScalableFloat|GetValueatLevel (LocalVariables|GetSF_TraceCapsuleHalfHeightCenter)) (ScalableFloat|GetValueatLevel (LocalVariables|GetSF_TraceCapsuleHalfHeightSide)) (ScalableFloat|GetValueatLevel (LocalVariables|GetSF_PropMaxHealthToDestroy)) (Math|Trig|Cos(Degrees) (ScalableFloat|GetValueasDouble (LocalVariables|GetSF_WallTiltMaxAngle))) (ScalableFloat|GetValueasDouble (LocalVariables|GetSF_ZOffsetTolerance)) "Component.ThrownDeployable.BlockingArea" "TraceTypeQuery6" (Variables|/ItemizedPropSpawnerParent/Blueprints/DataClass_ItemizedPropSpawnerData.DataClass_ItemizedPropSpawnerData_C|GetNativePropsToSpawn (Variables|/DeployedCoverGameplay/Items/ItemSpawner/B_DeployedCoverGrenade.B_DeployedCoverGrenade_C|GetData%20Asset _asb_deployed_cover_grenade)) (Utilities|Array|MakeArray "/Script/ItemizedPropSpawnerParentRuntime.ItemizedPropSpawnerWallChunk" "/Game/Athena/Items/EnvironmentalItems/HidingProps/B_HidingProp.B_HidingProp_C" "/Script/FortniteGame.BuildingContainer")) (Variables|GetTargetLocation) (Variables|GetYawRotation) (Variables|GetDebug)))
  (return _boutcanspawn _outspawninfo _outspawnrotation _outspawnlocation))
